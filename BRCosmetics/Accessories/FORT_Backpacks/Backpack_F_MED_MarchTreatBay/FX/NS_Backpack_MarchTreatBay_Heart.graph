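# Niagara system for the MarchTreatBay back bling that emits pink heart sprites reacting to event audio.
# A single emitter spawns short-lived particles in an offset, stretched shape, pulsing their size on a double-beat curve.
# Event audio amplitude and System.MusicReactivity drive spawn probability and sprite scale, with separate branches for front-end preview.
# Two sprite renderers are used, with derived larger, lighter glow size and color attributes.
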NiagaraSystem /BRCosmetics/Accessories/FORT_Backpacks/Backpack_F_MED_MarchTreatBay/FX/NS_Backpack_MarchTreatBay_Heart
dependencies {
 usedDataInterfaces [&NiagaraDataInterfaceCurve]
 usedDynamicInputs [&RandomRangeFloat]
 usedModules [
  &SystemState
  &self:SystemSpawnScript.SystemScriptSource.SystemScriptGraph.NiagaraNodeAssignment_0.SetVariables_#1
  &InterpolateOverTime
 ]
}
emitters {Heart {
  emitterUpdateScript [
   &EmitterState(Life Cycle Mode=::NewEnumerator1, Loop Duration=1, UseLoopDelay=false)
   &SpawnRate(
    Spawn Probability=&MakeCustomFloatFromBool(
     Bool=$User.bIsFrontEndPreview
     True Float=&Lerp_Float(A=0.3, B=1, Alpha=$NPC.FortniteNPC.EventAudio_Amplitude)
     False Float=&MakeCustomFloatFromBool(
      Bool=&SetBoolByFloatComparison(A=$NPC.FortniteNPC.EventAudio_Amplitude, B=0, Comparison Type=::NewEnumerator4)
      True Float=&Lerp_Float(A=0, B=1, Alpha=$System.MusicReactivity)
      False Float=1
     )
    )
    SpawnRate=2
    Use Spawn Probability=true
   )
  ]
  particleSpawnScript [
   &InitializeParticle(
    Color=<rgba 0.9427 0.0786 0.2053 3>
    Lifetime Max=1.5
    Lifetime Min=0.8
    Lifetime Mode=::NewEnumerator1
    Sprite Size Mode=::NewEnumerator1
    Uniform Sprite Size Max=8
    UsePositionOffset=true
    Write Material Random=true
   )
   &ShapeLocation(
    Handle Radius=2
    Large Radius=10
    MaskSpawnGroup=false
    Non Uniform Scale=<0.9 0.7 2>
    Offset=<0 -12 15>
    Offset Mode=::NewEnumerator0
    Override Seed=false
    Shape Primitive=::NewEnumerator3
    UseOwnerScale=true
   )
   &CameraOffset(Camera Offset Amount=2)
  ]
  particleUpdateScript [
   &ParticleState()
   &ScaleSpriteSize(
    Uniform Curve Scale=&MakeCustomFloatFromBool(
     Bool=$User.bIsFrontEndPreview
     True Float=&Lerp_Float(
      A=&Multiply_Float(A=0.5, B=&ClampFloat(Float=$NPC.FortniteNPC.EventAudio_Amplitude, Min=1, Max=1.3))
      B=1
      Alpha=$System.MusicReactivity
     )
     False Float=&Lerp_Float(
      A=&Multiply_Float(A=1, B=&ClampFloat(Float=$NPC.FortniteNPC.EventAudio_Amplitude, Min=0.9, Max=1.3))
      B=1
      Alpha=$System.MusicReactivity
     )
    )
    Uniform Curve Sprite Scale=DataInterface() {propertyValues {Curve {
       defaultValue=1
       keys [
        {arriveTangent=4.1399, time=0, value=0}
        {arriveTangent=0.875, time=0.1, value=1.1}
        {arriveTangent=-0.375, leaveTangent=-0.375, time=0.3, value=0.375}
        {arriveTangent=-1.4167, leaveTangent=-1.4167, time=0.4, value=1.05}
        {arriveTangent=0.0224, leaveTangent=0.0224, time=0.6, value=0.4}
        <at 0.7 1>
        <at 1 0>
       ]
      }}}
   )
   /SetVariables(Particles.Size_glow:Vector2f=&Multiply_Vector2D(A=$Particles.SpriteSize, B=<1.2 1.2>)) {isSetParameters=true}
   /SetVariables(Particles.Color_glow:LinearColor=&MakeLinearColorFromVectorAndFloat("Vector (RGB)"=&MakeVectorFromLinearColorRGB(Color=<rgba 1 0.2799 0.3951 1>), "Float (Alpha)"=0.8)) {isSetParameters=true}
   &SolveForcesAndVelocity(Clamp Velocity=false, Limit Acceleration=false)
  ]
  rendererClasses [&NiagaraSpriteRendererProperties]
  renderers [
   {rendererClass=&NiagaraSpriteRendererProperties}
   {rendererClass=&NiagaraSpriteRendererProperties, rendererIndex=1}
  ]
 }}
summary {
 systemName=NS_Backpack_MarchTreatBay_Heart
 emitters [{emitterName=Heart, rendererClasses [&NiagaraSpriteRendererProperties]}]
 userVariables [
  User.bIsFrontEnd {type=NiagaraBool}
  User.bIsFrontEndPreview {type=NiagaraBool}
 ]
}
