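# Material function for the MarchTreatBay backpack that animates its face by switching eye (open, closed, hearts) and mouth (open, closed) pieces.
# Vertex color masks select pieces, and World Position Offset toward the object position hides them, with a UV1-based mask for outlines.
# Layered time-based sines create blinks, while frontend-loop audio amplitude supplies a reactivity signal.
# A static switch substitutes debug parameters for both signals.
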MaterialFunction /BRCosmetics/Accessories/FORT_Backpacks/Backpack_F_MED_MarchTreatBay/Materials/MF_MarchTreatBay_Backpack
graph {nodes [
  n1 = GetMaterialAttributes(None=FunctionInput() {inputType=FunctionInput_MaterialAttributes}) {attributeGetTypes [#1]}
  n2 = Subtract(
   A=ObjectPositionWS() {originType=CameraRelative}
   B=Add(A=WorldPosition() {worldPositionShaderOffset=WPT_CameraRelative}.XYZ, B=@n1.WorldPositionOffset)
  )
  n3 = LinearInterpolate(A=Multiply(A=@n2, B=NamedRerouteUsage()), B=@n2, Alpha=NamedRerouteUsage())
  FunctionOutput(None=SetMaterialAttributes!(MaterialAttributes=@n1.MaterialAttributes, WorldPositionOffset=Add(A=@n1.WorldPositionOffset, B=@n3)) {attributeSetTypes [#1]})
  n4 = VertexColor()
  NamedRerouteDeclaration(None=ChannelMaskParameter(None=@n4) {group=MarchTreatExpressions, parameterName=EyesMask}.R) {name=EyeMask, nodeColor=<rgba 0 0.2031 1 1>}
  NamedRerouteDeclaration(None=ScalarParameter() {defaultValue=0.1, group=MarchTreatExpressions, parameterName=Eyes_Open_Value}) {name=Eyes_Open_Value, nodeColor=<rgba 1 0 0.8438 1>}
  n5 = MaterialFunctionCall!(Hide Element Toggle=OneMinus(None=NamedRerouteUsage()), VertexColorChannel=NamedRerouteUsage(), Vertex Value to hide=NamedRerouteUsage()) {materialFunction=/Game/Packages/Fortress_SharedMaterials/Base_Material_Functions/VertexColorMask}
  n6 = MaterialFunctionCall(Hide Element Toggle=NamedRerouteUsage(), VertexColorChannel=NamedRerouteUsage(), Vertex Value to hide=NamedRerouteUsage()) {materialFunction=&VertexColorMask}
  n7 = Constant() {r=1}
  n8 = NamedRerouteUsage()
  n9 = NamedRerouteUsage()
  n10 = OneMinus(None=NamedRerouteUsage())
  n11 = NamedRerouteUsage()
  n12 = NamedRerouteUsage()
  n13 = NamedRerouteUsage()
  n14 = NamedRerouteUsage()
  n15 = NamedRerouteUsage()
  n16 = Max(
   A=LinearInterpolate(
    A=Max(A=@n5.Result, B=@n6.Result)
    B=MaterialFunctionCall(Hide Element Toggle=@n7, VertexColorChannel=@n8, Vertex Value to hide=@n9) {materialFunction=&VertexColorMask}.Result
    Alpha=NamedRerouteUsage()
   )
   B=Max(
    A=MaterialFunctionCall(Hide Element Toggle=@n10, VertexColorChannel=@n11, Vertex Value to hide=@n12) {materialFunction=&VertexColorMask}.Result
    B=MaterialFunctionCall(Hide Element Toggle=@n13, VertexColorChannel=@n14, Vertex Value to hide=@n15) {materialFunction=&VertexColorMask}.Result
   )
  )
  n17 = Max(A=NamedRerouteUsage(), B=NamedRerouteUsage())
  NamedRerouteDeclaration(None=Multiply(A=OneMinus(None=@n16), B=Ceil(None=@n17))) {name=HiddenMask, nodeColor=<rgba 0.7813 0 1 1>}
  n18 = Time()
  n19 = Multiply(A=@n18) {constB=0.7}
  n20 = Multiply(A=Sine(None=@n18) {period=0.6}, B=Saturate(None=Sine(None=@n19)))
  n21 = Cosine(None=Multiply(A=@n18) {constB=0.1})
  n22 = SmoothStep(Value=Multiply(A=Saturate(None=@n20), B=Saturate(None=@n21))) {constMin=0.4}
  NamedRerouteDeclaration(None=StaticSwitchParameter(True=ScalarParameter() {group=MarchTreatExpressions, parameterName=DebugEyesClosed}, False=Ceil(None=@n22)) {group=MarchTreatExpressions, parameterName=DebugExpressions}) {name=BlinksSine, nodeColor=<rgba 0.875 1 0 1>}
  n23 = ScalarParameter() {group=MarchTreatExpressions, parameterName=DebugReactivity}
  n24 = Ceil(None=Power(
    Base=MaterialFunctionCall!(InAmplitude, InSpectrum, CustomFakeSpectrum, CustomFakeAmplitude) {materialFunction=/BRCosmetics/Global/MaterialFunctions/AudioReactivity/MF_AudioReactivity_FrontendLoop}.Amplitude
    Exp=ScalarParameter() {defaultValue=0.01, parameterName=AmplitudeExponent}
   ) {constExponent=0.1})
  NamedRerouteDeclaration(None=Multiply(
    A=StaticSwitchParameter(True=@n23, False=@n24) {group=MarchTreatExpressions, parameterName=DebugExpressions}
    B=ScalarParameter() {defaultValue=1, group=MarchTreatExpressions, parameterName=Is Reactive}
   )) {name=Amplitude, nodeColor=<rgba 0.8047 0 1 1>}
  n25 = ChannelMaskParameter(None=@n4) {group=MarchTreatExpressions, maskChannel=Blue, parameterName=MouthMask, defaultValue=<rgba 0 0 1 0>}
  NamedRerouteDeclaration(None=@n25.R) {name=MouthMask, nodeColor=<rgba 1 0 0.0343 1>}
  NamedRerouteDeclaration(None=ScalarParameter() {defaultValue=0.2, group=MarchTreatExpressions, parameterName=Mouth_Closed}) {name=Mouth_Closed, nodeColor=<rgba 1 0.4492 0.688 1>}
  NamedRerouteDeclaration(None=ScalarParameter() {defaultValue=0.1, group=MarchTreatExpressions, parameterName=Mouth_Open}) {name=Mouth_Open, nodeColor=<rgba 0 1 0.1409 1>}
  NamedRerouteDeclaration(None=ScalarParameter() {defaultValue=0.3, group=MarchTreatExpressions, parameterName=Eyes_Closed_Value}) {name=Eyes_Closed_Value, nodeColor=<rgba 1 0.307 0.3332 1>}
  NamedRerouteDeclaration(None=ScalarParameter() {defaultValue=0.2, group=MarchTreatExpressions, parameterName=Eyes_Hearts_Value}) {name=Eyes_Hearts_Value, nodeColor=<rgba 0 0.4242 1 1>}
  n26 = ComponentMask(None=TextureCoordinate() {coordinateIndex=1}) {r=true}
  n27 = Saturate(None=Sign(None=Subtract(A=@n26) {constB=0.0001}))
  NamedRerouteDeclaration(None=Multiply(
    A=NamedRerouteDeclaration(None=@n27) {name=UV2Mask, nodeColor=<rgba 0 1 0.9063 1>}
    B=ScalarParameter() {group=MarchTreatExpressions, parameterName=IsOutline}
   )) {name=FaceOutlineMask, nodeColor=<rgba 1 0.5859 0 1>}
 ]}
