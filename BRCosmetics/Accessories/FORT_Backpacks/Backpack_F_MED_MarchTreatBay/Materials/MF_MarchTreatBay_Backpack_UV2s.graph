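# Material function for the MarchTreatBay backpack that overlays expression color and normal textures, mapped on the second UV channel, onto incoming lit color, shaded color and normal.
# A binary mask from that channel's U coordinate, positive beyond a near-zero threshold, confines the overlay to UV-mapped areas.
# Lerps blend each output, with the shaded overlay derived by a tint multiply.
# The Mask input feeds nothing; the named mask reroute takes the computed value directly.
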MaterialFunction /BRCosmetics/Accessories/FORT_Backpacks/Backpack_F_MED_MarchTreatBay/Materials/MF_MarchTreatBay_Backpack_UV2s
graph {nodes [
  n1 = ComponentMask(None=TextureCoordinate() {coordinateIndex=1}) {r=true}
  n2 = Saturate(None=Sign(None=Subtract(A=@n1) {constB=0.0001}))
  FunctionInput(Preview=@n2) {bUsePreviewValueAsDefault=true, inputName=Mask, inputType=FunctionInput_Scalar}
  n3 = NamedRerouteUsage()
  n4 = TextureSampleParameter2D(UVs=NamedRerouteUsage()) {
   parameterName=UV2ColorLit
   texture=/BRCosmetics/Accessories/FORT_Backpacks/Backpack_F_MED_MarchTreatBay/Textures/T_MarchTreatBay_Expressions_CL
  }
  n5 = NamedRerouteUsage()
  FunctionOutput(None=NamedRerouteDeclaration(None=LinearInterpolate(A=@n3, B=@n4.RGB, Alpha=@n5)) {name=ColorLit, nodeColor=<rgba 1 0.5859 0 1>}) {outputName=OutColorLit}
  n6 = TextureSampleParameter2D(UVs=NamedRerouteUsage()) {
   parameterName=UV2Normal
   samplerType=SAMPLERTYPE_Normal
   texture=/BRCosmetics/Accessories/FORT_Backpacks/Backpack_F_MED_MarchTreatBay/Textures/T_MarchTreatBay_Expressions_N
  }
  FunctionOutput(None=LinearInterpolate(A=NamedRerouteUsage(), B=@n6.RGB, Alpha=NamedRerouteUsage())) {outputName=OutNormal}
  n7 = NamedRerouteUsage()
  n8 = VectorParameter() {defaultValue=<rgba 0.7083 0.5682 0.6148 1>}
  FunctionOutput(None=LinearInterpolate(A=NamedRerouteUsage(), B=Multiply(A=@n7, B=@n8.RGB) {constB=0.75}, Alpha=NamedRerouteUsage())) {outputName=OutColoShaded}
  NamedRerouteDeclaration(None=@n2) {name=UV2Mask, nodeColor=<rgba 0 0.2031 1 1>}
  NamedRerouteDeclaration(None=FunctionInput() {inputName=InColorLit}) {name=InColorLit, nodeColor=<rgba 1 0.023 0.877 1>}
  NamedRerouteDeclaration(None=FunctionInput() {inputName=InNormal}) {name=InNormal, nodeColor=<rgba 0.2747 0 1 1>}
  NamedRerouteDeclaration(None=TextureCoordinate() {coordinateIndex=1}) {name=UVs, nodeColor=<rgba 0 0.5781 1 1>}
  NamedRerouteDeclaration(None=FunctionInput() {inputName=InColorShaded}) {name=InColorShaded, nodeColor=<rgba 0.1081 1 0.622 1>}
  OneMinus(None=@n2)
 ]}
