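# Animation Blueprint for the MarchTreatBay backpack cosmetic, adding secondary motion and adjusting the pack as the wearer's spine bends.
# Each update maps spine_05-to-spine_01 yaw delta to an additive pack bone offset and sets a simulation-space variable from bIsInFrontEnd.
# A spring-driven AnimDynamics chain moves legs, bows, hand and heard_dyn into a cached pose.
# Two differently tuned wing and head chains branch from it, selected by bIsInFrontEnd.
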AnimBlueprint /BRCosmetics/Accessories/FORT_Backpacks/Backpack_F_MED_MarchTreatBay/Meshes/F_MED_MarchTreatBay_Pack_AnimBP
target_skeleton=&/F_MED_MarchTreatBay_Pack_Skeleton
graphs [
 AnimGraph {nodes [
   n1 = Anim Dynamics - Bone: wing_dyn_l!(
    SimulationSpace:AnimPhysSimSpaceType Enum
    ComponentPose=Local To Component(LocalPose=Use cached pose 'Main pose'())
    Alpha=1
    GravityOverride:Vector="0,0,-500"
    ExternalForce="0,0,0"
   ) -> (Pose) {
    angularDampingOverride=1
    angularSpringConstant=1500
    bAngularSpring=true
    bGravityOverrideInSimSpace=true
    bOverrideAngularDamping=true
    bUseGravityOverride=true
    alphaBoolBlend {blendOption=Linear}
    boundBone {boneName=wing_dyn_l}
    componentAppliedLinearAccClamp=<3000 3000 2500>
    componentLinearAccScale=<0.6 0.6 0.6>
    componentLinearVelScale=<0.6 0.6 0.6>
    physicsBodyDefinitions [{
      boundBone {boneName=wing_dyn_l}
      boxExtents=<14 14 14>
      constraintSetup {coneAngle=1, angularLimitsMax=<2 0 30>, angularLimitsMin=<-2 0 -10>, angularTarget=<1 0 0>}
      localJointOffset=<1 0 0>
     }]
    simSpaceSettings {maxAngularAcceleration=3000, maxAngularVelocity=3000, simSpaceAngularAlpha=0.1}
   }
   n2 = Anim Dynamics - Bone: wing_dyn_r!(SimulationSpace:AnimPhysSimSpaceType Enum, ComponentPose=@n1, Alpha=1, GravityOverride:Vector="0,0,-500", ExternalForce="0,0,0") -> (Pose) {
    angularDampingOverride=1
    angularSpringConstant=1500
    bAngularSpring=true
    bGravityOverrideInSimSpace=true
    bOverrideAngularDamping=true
    bUseGravityOverride=true
    alphaBoolBlend {blendOption=Linear}
    boundBone {boneName=wing_dyn_r}
    componentAppliedLinearAccClamp=<3000 3000 2500>
    componentLinearAccScale=<0.6 0.6 0.6>
    componentLinearVelScale=<0.6 0.6 0.6>
    physicsBodyDefinitions [{
      boundBone {boneName=wing_dyn_r}
      boxExtents=<14 14 14>
      constraintSetup {coneAngle=1, angularLimitsMax=<2 0 10>, angularLimitsMin=<-2 0 -60>, angularTarget=<1 0 0>}
      localJointOffset=<1 0 0>
     }]
    simSpaceSettings {maxAngularAcceleration=3000, maxAngularVelocity=3000, simSpaceAngularAlpha=0.1}
   }
   n3 = Anim Dynamics - Bone: head_dyn!(SimulationSpace:AnimPhysSimSpaceType Enum, ComponentPose=@n2, Alpha=0.9, GravityOverride:Vector="0,0,-100", ExternalForce="0,0,0") -> (Pose) {
    angularDampingOverride=1
    angularSpringConstant=1500
    bAngularSpring=true
    bGravityOverrideInSimSpace=true
    bOverrideAngularDamping=true
    bUseGravityOverride=true
    alphaBoolBlend {blendOption=Linear}
    boundBone {boneName=head_dyn}
    componentAppliedLinearAccClamp=<3000 3000 2500>
    componentLinearAccScale=<0.2 0.2 0.2>
    componentLinearVelScale=<0.2 0.2 0.2>
    physicsBodyDefinitions [{
      boundBone {boneName=head_dyn}
      boxExtents=<12 12 12>
      constraintSetup {coneAngle=1, angularLimitsMax=<1.5 7 4.5>, angularLimitsMin=<-1.5 -22 -4.5>, angularTarget=<1 0 0>}
      localJointOffset=<1 0 0>
     }]
    simSpaceSettings {maxAngularAcceleration=3000, maxAngularVelocity=3000, simSpaceAngularAlpha=0.1}
   }
   n4 = Anim Dynamics - Bone: wing_dyn_l(ComponentPose=Local To Component(LocalPose=Use cached pose 'Main pose'()), Alpha=1, GravityOverride="0,0,-500", ExternalForce="0,0,0") {
    angularDampingOverride=1
    angularSpringConstant=1500
    bAngularSpring=true
    bGravityOverrideInSimSpace=true
    bOverrideAngularDamping=true
    bUseGravityOverride=true
    alphaBoolBlend {blendOption=Linear}
    boundBone {boneName=wing_dyn_l}
    componentAppliedLinearAccClamp=<3000 3000 2500>
    componentLinearAccScale=<0.7 0.7 0.7>
    componentLinearVelScale=<0.7 0.7 0.7>
    physicsBodyDefinitions [{
      boundBone {boneName=wing_dyn_l}
      boxExtents=<10 10 11>
      constraintSetup {coneAngle=1, angularLimitsMax=<2 0 30>, angularLimitsMin=<-2 0 -10>, angularTarget=<1 0 0>}
      localJointOffset=<1 0 0>
     }]
    simSpaceSettings {maxAngularAcceleration=3000, maxAngularVelocity=3000, simSpaceAngularAlpha=0.1}
   }
   n5 = Anim Dynamics - Bone: wing_dyn_r(ComponentPose=@n4, Alpha=1, GravityOverride="0,0,-500", ExternalForce="0,0,0") {
    angularDampingOverride=1
    angularSpringConstant=1500
    bAngularSpring=true
    bGravityOverrideInSimSpace=true
    bOverrideAngularDamping=true
    bUseGravityOverride=true
    alphaBoolBlend {blendOption=Linear}
    boundBone {boneName=wing_dyn_r}
    componentAppliedLinearAccClamp=<3000 3000 2500>
    componentLinearAccScale=<0.7 0.7 0.7>
    componentLinearVelScale=<0.7 0.7 0.7>
    physicsBodyDefinitions [{
      boundBone {boneName=wing_dyn_r}
      boxExtents=<10 10 10>
      constraintSetup {coneAngle=1, angularLimitsMax=<2 0 10>, angularLimitsMin=<-2 0 -60>, angularTarget=<1 0 0>}
      localJointOffset=<1 0 0>
     }]
    simSpaceSettings {maxAngularAcceleration=3000, maxAngularVelocity=3000, simSpaceAngularAlpha=0.1}
   }
   n6 = Anim Dynamics - Bone: head_dyn(ComponentPose=@n5, Alpha=1, GravityOverride="0,0,-100", ExternalForce="0,0,0") {
    angularDampingOverride=1
    angularSpringConstant=1500
    bAngularSpring=true
    bGravityOverrideInSimSpace=true
    bOverrideAngularDamping=true
    bUseGravityOverride=true
    alphaBoolBlend {blendOption=Linear}
    boundBone {boneName=head_dyn}
    componentAppliedLinearAccClamp=<3000 3000 2500>
    componentLinearAccScale=<0.2 0.2 0.2>
    componentLinearVelScale=<0.2 0.2 0.2>
    physicsBodyDefinitions [{
      boundBone {boneName=head_dyn}
      boxExtents=<12 12 12>
      constraintSetup {coneAngle=1, angularLimitsMax=<2 10 5>, angularLimitsMin=<-2 -30 -5>, angularTarget=<1 0 0>}
      localJointOffset=<1 0 0>
     }]
    simSpaceSettings {maxAngularAcceleration=3000, maxAngularVelocity=3000, simSpaceAngularAlpha=0.1}
   }
   Output Pose(Result=Blend Poses by bool(BlendPose_0=Component To Local(ComponentPose=@n3), BlendPose_1=Component To Local(ComponentPose=@n6), bActiveValue=Get bIsInFrontEnd())) {layerGroup=None}
   n7 = "Transform (Modify) Bone - Bone: pack"(ComponentPose=Local To Component(LocalPose=Input Pose()), Translation=Get backpack_translation(), Rotation=Get backpack_rotate()) {rotationMode=BMM_Additive, translationMode=BMM_Additive, alphaBoolBlend {blendOption=Linear}, boneToModify {boneName=pack}}
   n8 = Anim Dynamics - Bone: leg_dyn_l!(SimulationSpace:AnimPhysSimSpaceType Enum, ComponentPose=@n7, Alpha=1, GravityOverride:Vector="0,0,-500", ExternalForce="0,0,0") -> (Pose) {
    angularDampingOverride=1
    angularSpringConstant=1000
    bAngularSpring=true
    bGravityOverrideInSimSpace=true
    bOverrideAngularDamping=true
    bUseGravityOverride=true
    alphaBoolBlend {blendOption=Linear}
    boundBone {boneName=leg_dyn_l}
    componentAppliedLinearAccClamp=<3000 3000 2500>
    componentLinearAccScale=<0.3 0.3 0.3>
    componentLinearVelScale=<0.3 0.3 0.3>
    physicsBodyDefinitions [{
      boundBone {boneName=leg_dyn_l}
      boxExtents=<14 14 14>
      constraintSetup {coneAngle=1, angularLimitsMax=<2 20 5>, angularLimitsMin=<-2 -30 -20>, angularTarget=<1 0 0>}
      localJointOffset=<1 0 0>
     }]
    simSpaceSettings {maxAngularAcceleration=3000, maxAngularVelocity=3000, simSpaceAngularAlpha=0.1}
   }
   n9 = Anim Dynamics - Bone: reg_dyn_r!(SimulationSpace:AnimPhysSimSpaceType Enum, ComponentPose=@n8, Alpha=1, GravityOverride:Vector="0,0,-500", ExternalForce="0,0,0") -> (Pose) {
    angularDampingOverride=1
    angularSpringConstant=1000
    bAngularSpring=true
    bGravityOverrideInSimSpace=true
    bOverrideAngularDamping=true
    bUseGravityOverride=true
    alphaBoolBlend {blendOption=Linear}
    boundBone {boneName=reg_dyn_r}
    componentAppliedLinearAccClamp=<3000 3000 2500>
    componentLinearAccScale=<0.4 0.4 0.4>
    componentLinearVelScale=<0.4 0.4 0.4>
    physicsBodyDefinitions [{
      boundBone {boneName=reg_dyn_r}
      boxExtents=<12 12 12>
      constraintSetup {coneAngle=1, angularLimitsMax=<2 20 20>, angularLimitsMin=<-2 -30 -5>, angularTarget=<1 0 0>}
      localJointOffset=<1 0 0>
     }]
    simSpaceSettings {maxAngularAcceleration=3000, maxAngularVelocity=3000, simSpaceAngularAlpha=0.1}
   }
   n10 = Anim Dynamics - Bone: bow_dyn_l!(SimulationSpace:AnimPhysSimSpaceType Enum, ComponentPose=@n9, Alpha=1, GravityOverride:Vector="0,0,-100", ExternalForce="0,0,0") -> (Pose) {
    angularDampingOverride=1
    angularSpringConstant=1500
    bAngularSpring=true
    bGravityOverrideInSimSpace=true
    bOverrideAngularDamping=true
    bUseGravityOverride=true
    alphaBoolBlend {blendOption=Linear}
    boundBone {boneName=bow_dyn_l}
    componentAppliedLinearAccClamp=<3000 3000 2500>
    componentLinearAccScale=<0.5 0.5 0.5>
    componentLinearVelScale=<0.5 0.5 0.5>
    physicsBodyDefinitions [{
      boundBone {boneName=bow_dyn_l}
      boxExtents=<10 10 10>
      constraintSetup {coneAngle=1, angularLimitsMax=<2 30 5>, angularLimitsMin=<-2 -30 -5>, angularTarget=<1 0 0>}
      localJointOffset=<1 0 0>
     }]
    simSpaceSettings {maxAngularAcceleration=3000, maxAngularVelocity=3000, simSpaceAngularAlpha=0.1}
   }
   n11 = Anim Dynamics - Bone: bow_dyn_r!(SimulationSpace:AnimPhysSimSpaceType Enum, ComponentPose=@n10, Alpha=1, GravityOverride:Vector="0,0,-100", ExternalForce="0,0,0") -> (Pose) {
    angularDampingOverride=1
    angularSpringConstant=1500
    bAngularSpring=true
    bGravityOverrideInSimSpace=true
    bOverrideAngularDamping=true
    bUseGravityOverride=true
    alphaBoolBlend {blendOption=Linear}
    boundBone {boneName=bow_dyn_r}
    componentAppliedLinearAccClamp=<3000 3000 2500>
    componentLinearAccScale=<0.2 0.2 0.2>
    componentLinearVelScale=<0.2 0.2 0.2>
    physicsBodyDefinitions [{
      boundBone {boneName=bow_dyn_r}
      boxExtents=<12 12 12>
      constraintSetup {coneAngle=1, angularLimitsMax=<2 30 5>, angularLimitsMin=<-2 -30 -5>, angularTarget=<1 0 0>}
      localJointOffset=<1 0 0>
     }]
    simSpaceSettings {maxAngularAcceleration=3000, maxAngularVelocity=3000, simSpaceAngularAlpha=0.1}
   }
   n12 = Anim Dynamics - Bone: heard_dyn!(SimulationSpace:AnimPhysSimSpaceType Enum, ComponentPose=@n11, Alpha=1, GravityOverride:Vector="0,0,-100", ExternalForce="0,0,0") -> (Pose) {
    angularDampingOverride=1
    angularSpringConstant=1500
    bAngularSpring=true
    bGravityOverrideInSimSpace=true
    bOverrideAngularDamping=true
    bUseGravityOverride=true
    alphaBoolBlend {blendOption=Linear}
    boundBone {boneName=heard_dyn}
    componentAppliedLinearAccClamp=<3000 3000 2500>
    componentLinearAccScale=<0.2 0.2 0.2>
    componentLinearVelScale=<0.2 0.2 0.2>
    physicsBodyDefinitions [{
      boundBone {boneName=heard_dyn}
      boxExtents=<12 12 12>
      constraintSetup {coneAngle=1, angularLimitsMax=<0 0 20>, angularLimitsMin=<0 0 -20>, angularTarget=<1 0 0>}
      localJointOffset=<1 0 0>
     }]
    simSpaceSettings {maxAngularAcceleration=3000, maxAngularVelocity=3000, simSpaceAngularAlpha=0.1}
   }
   n13 = Anim Dynamics - Bone: hand_dyn_l!(SimulationSpace:AnimPhysSimSpaceType Enum, ComponentPose=@n12, Alpha=1, GravityOverride:Vector="0,0,-100", ExternalForce="0,0,0") -> (Pose) {
    angularDampingOverride=1
    angularSpringConstant=3000
    bAngularSpring=true
    bGravityOverrideInSimSpace=true
    bOverrideAngularDamping=true
    bUseGravityOverride=true
    alphaBoolBlend {blendOption=Linear}
    boundBone {boneName=hand_dyn_l}
    componentAppliedLinearAccClamp=<3000 3000 2500>
    componentLinearAccScale=<0.1 0.1 0.1>
    componentLinearVelScale=<0.1 0.1 0.1>
    physicsBodyDefinitions [{
      boundBone {boneName=hand_dyn_l}
      boxExtents=<20 20 20>
      constraintSetup {coneAngle=1, angularLimitsMax=<2 20 10>, angularLimitsMin=<-2 -30 -20>, angularTarget=<1 0 0>}
      localJointOffset=<1 0 0>
     }]
    simSpaceSettings {maxAngularAcceleration=3000, maxAngularVelocity=3000, simSpaceAngularAlpha=0.1}
   }
   Save cached pose 'Main pose'(Pose=Component To Local(ComponentPose=@n13)) {cacheName=Main pose}
  ]}
 EventGraph {nodes [
   Event BlueprintUpdateAnimation()
   n1 = GetOwnerSkeletalMeshComponent()
   n2 = BreakRotator(InRot=BreakTransform(InTransform=GetDeltaTransformFromRefPose(self=@n1, BoneName=spine_05, BaseName=spine_01)).Rotation)
   n3 = MapRangeClamped(Value=@n2.Yaw, InRangeA=-30, InRangeB=-150, OutRangeB=220)
   > Set backpack_rotate(backpack_rotate=MakeRotator(Roll=float + float(B=@n3)))
   > Set backpack_translation(backpack_translation=MakeVector(
     Y=MapRangeClamped(Value=@n2.Yaw, InRangeA=-20, InRangeB=-50, OutRangeB=-12)
     Z=MapRangeClamped(Value=@n2.Yaw, InRangeA=-30, InRangeB=-50, OutRangeB=-2.5)
    ))
   > n4 = Branch(Condition=Get bIsInFrontEnd())
   > Set SimlationSpace(SimlationSpace:AnimPhysSimSpaceType Enum=Component) -> (Output_Get:AnimPhysSimSpaceType Enum)
   Set SimlationSpace(execute=@n4.else)
   TryGetPawnOwner()
   FInterpTo(Current=0, Target=0, DeltaTime=0, InterpSpeed=10)
   MapRangeClamped(Value=@n2.Yaw, InRangeA=-8, InRangeB=-30, OutRangeB=10)
  ]}
]
variables [
 backpack_rotate {type=Rotator}
 backpack_translation {type=Vector}
 SimlationSpace {type=AnimPhysSimSpaceType}
]
